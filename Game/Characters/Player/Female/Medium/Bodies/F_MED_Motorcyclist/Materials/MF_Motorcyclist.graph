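# MF_Motorcyclist is a material function that layers skin effects and a recolor onto the female medium Motorcyclist player body.
# Incoming material attributes run through shared weapon-skin thin film and emissive fresnel functions, optional static-switched panning emissive, then MF_MetalsLUT.
# Base color finally lerps toward a MotorcyclistDiffuseColor parameter.
# One SkinFX_Mask texture drives the masking, with static component-mask parameters choosing channels for thin film, panning emissive and recolor.
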MaterialFunction /Game/Characters/Player/Female/Medium/Bodies/F_MED_Motorcyclist/Materials/MF_Motorcyclist
graph {nodes [
  CameraVectorWS()
  FunctionInput() {inputName=PreSkinnedLocalPosition}
  n1 = TextureSample(Tex=NamedRerouteDeclaration(None=TextureObjectParameter() {group=Skin FX, parameterName=SkinFX_Mask}) {name=SkinFX_Mask, nodeColor=<rgba 0.9842 0 1 1>})
  n2 = MaterialFunctionCall!(In=MaterialFunctionCall!(
    Alpha=StaticComponentMaskParameter(None=@n1.RGB) {defaultR=true, group=Skin FX, parameterName=ThinFilm_Channel}
    Material Attributes=FunctionInput() {inputName=Material, inputType=FunctionInput_MaterialAttributes, sortPriority=-20}
   ) {materialFunction=/Game/Weapons/WeaponSkins/MaterialFunction/ThinFilmFunction}.Result) {materialFunction=/Game/Weapons/WeaponSkins/MaterialFunction/MF_EmissiveFresnel}
  n3 = MaterialFunctionCall!(
   Alpha=StaticComponentMaskParameter(None=@n1.RGB) {defaultR=true, group=Skin FX, parameterName=PanningEmissive_Channel}
   isVehicle
   Gmap
   UVOffset
   Local Normal
   Local Position
   EmissiveColorOverride
   Material=@n2.Result
  ) {materialFunction=/Game/Weapons/WeaponSkins/MaterialFunction/PanningEmissive}
  n4 = GetMaterialAttributes(None=MaterialFunctionCall(Mask=@n1.RGBA, Materials=StaticSwitchParameter(True=@n3.Result, False=@n2.Result) {group=Skin FX, parameterName=Use Panning Emissive}) {materialFunction=&MF_MetalsLUT}.Result) {attributeGetTypes [#1, #2]}
  n5 = VectorParameter() {group=MotorCyclist, parameterName=MotorcyclistDiffuseColor, sortPriority=10}
  n6 = StaticComponentMaskParameter(None=@n1.RGB) {defaultB=true, group=MotorCyclist, parameterName=MotorcyclistDiffuse_Channel, sortPriority=0}
  FunctionOutput(None=SetMaterialAttributes!(
    MaterialAttributes=@n4.MaterialAttributes
    EmissiveColor=@n4.MaterialAttributes
    BaseColor=LinearInterpolate(A=@n4.BaseColor, B=@n5.RGB, Alpha=@n6)
   ) {attributeSetTypes [#1, #2]})
  PixelNormalWS()
  VertexNormalWS()
 ]}
